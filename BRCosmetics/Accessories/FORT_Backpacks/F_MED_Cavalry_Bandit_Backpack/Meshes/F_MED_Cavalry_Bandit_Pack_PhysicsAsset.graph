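# Physics asset for the Cavalry Bandit backpack cosmetic, defining collision and joint limits for its dangling straps, flap and two grenades.
# A box pack body anchors a four-capsule back strap chain, a two-link front strap, the flap and both grenades.
# Swing and twist limits are tight on grenades and strap bases, looser along the chains and mostly single-axis on the flap.
# Two large offset capsules on the unconstrained root add broad collision volume.
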PhysicsAsset /BRCosmetics/Accessories/FORT_Backpacks/F_MED_Cavalry_Bandit_Backpack/Meshes/F_MED_Cavalry_Bandit_Pack_PhysicsAsset
bodies {
 dyn_bk_strap_01 [{shapeType=Capsule, length=1.5, radius=1.5, shapeName=dyn_bk_strap_01_capsule, center=<2 0 0>, rotation=<rot 90 -180 -180>}]
 dyn_bk_strap_02 [{shapeType=Capsule, length=3, radius=1.5, shapeName=dyn_bk_strap_02_capsule, center=<2.7682 0 0>, rotation=<rot 90 -180 -180>}]
 dyn_bk_strap_03 [{shapeType=Capsule, length=3.5, radius=1.5, shapeName=dyn_bk_strap_03_capsule, center=<2.8538 0 0>, rotation=<rot 90 -180 -180>}]
 dyn_bk_strap_04 [{shapeType=Capsule, length=2.75, radius=1.5, shapeName=dyn_bk_strap_04_capsule, center=<1.4 0 0>, rotation=<rot 0 -180 -180>}]
 dyn_fr_flap [{shapeType=Box, extentX=4.7308, extentY=1.01, extentZ=10.6348, center=<5.5256 0 0>, rotation=<rot 90 0 0>}]
 dyn_fr_strap_01 [
  {shapeType=Box, extentX=4.7308, extentY=1.2882, extentZ=5.8375, center=<7.3112 -0.3479 0>, rotation=<rot 90 0 0>}
  {shapeType=Box, extentX=3.2334, extentY=0.538, extentZ=4.3842, shapeName=dyn_fr_strap_01_box, center=<2.5592 -0.5834 0>}
 ]
 dyn_fr_strap_02 [{shapeType=Box, extentX=4, extentY=1.1, extentZ=5, center=<2.5 0 0>, rotation=<rot 90 0 0>}]
 dyn_grenade_a [{shapeType=Capsule, length=5.3215, radius=2.6144, center=<2.0852 1.9494 0.2646>, rotation=<rot 0 -0.6886 -70.1815>}]
 dyn_grenade_b [{shapeType=Capsule, length=5.6544, radius=1.7728, center=<2.4556 1.0112 -0.2524>, rotation=<rot 0 -2.1431 -69.0538>}]
 pack [{shapeType=Box, extentX=17.2536, extentY=9.8229, extentZ=25.5764, center=<1.0094 -3.9386 -2.654>, rotation=<rot 69.1738 180 180>}]
 root [
  {shapeType=Capsule, length=26, radius=14.8, shapeName=root_capsule, center=<0 14 -15>}
  {shapeType=Capsule, length=26, radius=14.8, shapeName=root_capsule, center=<0 14 0>}
 ]
}
constraints [
 {bone1Name=dyn_bk_strap_01, bone2Name=pack, swing1LimitDegrees=5, swing2LimitDegrees=5, twistLimitDegrees=5}
 {bone1Name=dyn_bk_strap_02, bone2Name=dyn_bk_strap_01, swing1LimitDegrees=45, swing2LimitDegrees=45, twistLimitDegrees=15}
 {bone1Name=dyn_bk_strap_03, bone2Name=dyn_bk_strap_02, swing1LimitDegrees=45, swing2LimitDegrees=45, twistLimitDegrees=15}
 {bone1Name=dyn_bk_strap_04, bone2Name=dyn_bk_strap_03, swing1LimitDegrees=45, swing2LimitDegrees=10, twistLimitDegrees=5}
 {bone1Name=dyn_fr_strap_01, bone2Name=pack, swing1LimitDegrees=5, swing2LimitDegrees=5, twistLimitDegrees=2.5}
 {bone1Name=dyn_fr_strap_02, bone2Name=dyn_fr_strap_01, swing1LimitDegrees=40, swing2LimitDegrees=40, twistLimitDegrees=2.5}
 {bone1Name=dyn_grenade_a, bone2Name=pack, swing1LimitDegrees=5, swing2LimitDegrees=5, twistLimitDegrees=3}
 {bone1Name=dyn_grenade_b, bone2Name=pack, swing1LimitDegrees=7.5, swing2LimitDegrees=10, twistLimitDegrees=4}
 {bone1Name=dyn_fr_flap, bone2Name=pack, swing1LimitDegrees=45, swing2LimitDegrees=2, twistLimitDegrees=12}
]
