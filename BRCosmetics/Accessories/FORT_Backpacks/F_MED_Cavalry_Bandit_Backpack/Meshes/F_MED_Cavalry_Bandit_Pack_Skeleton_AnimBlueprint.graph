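# Animation Blueprint for the Cavalry Bandit backpack cosmetic, adding physics motion and bone offsets to the pack.
# A RigidBody node simulates the pack's physics asset with overridden gravity, then additive bone transforms nudge the pack and root bones.
# Three chained strap bones receive increasing additive yaw corrections, enabled only while skydiving with the parachute closed, as computed each update in the EventGraph.
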AnimBlueprint /BRCosmetics/Accessories/FORT_Backpacks/F_MED_Cavalry_Bandit_Backpack/Meshes/F_MED_Cavalry_Bandit_Pack_Skeleton_AnimBlueprint
target_skeleton=&/F_MED_Cavalry_Bandit_Pack_Skeleton
graphs [
 AnimGraph {nodes [
   n1 = RigidBody(ComponentPose=Local To Component(LocalPose=Input Pose())) {
    bClampLinearTranslationLimitToRefPose=true
    bOverrideWorldGravity=true
    overridePhysicsAsset=&/F_MED_Cavalry_Bandit_Pack_PhysicsAsset
    alphaBoolBlend {blendOption=Linear}
    componentAppliedLinearAccClamp=<2000 2000 2000>
    componentLinearAccScale=<0.4 0.4 0.4>
    componentLinearVelScale=<0.4 0.4 0.4>
    overrideWorldGravity {z=-980}
   }
   n2 = "Transform (Modify) Bone - Bone: pack"!(ComponentPose=@n1, bAlphaBoolEnabled:bool, Translation="0,-1.3,0", Rotation="0,0,0", Scale) -> (Pose) {
    alphaInputType=Bool
    translationMode=BMM_Additive
    translationSpace=BCS_BoneSpace
    alphaBoolBlend {blendOption=Linear}
    boneToModify {boneName=pack}
    translation {y=-1.3}
   }
   n3 = Get Strap_rot-_correction() -> (Strap_rot-_correction:bool)
   n4 = "Transform (Modify) Bone - Bone: dyn_bk_strap_01"!(ComponentPose=@n2, bAlphaBoolEnabled:bool=@n3, Translation="0,0,0", Rotation="0,-5,0", Scale="1,1,1") -> (Pose) {
    alphaInputType=Bool
    rotationMode=BMM_Additive
    rotationSpace=BCS_BoneSpace
    translationSpace=BCS_BoneSpace
    alphaBoolBlend {blendOption=Linear}
    boneToModify {boneName=dyn_bk_strap_01}
    rotation {yaw=-5}
   }
   n5 = "Transform (Modify) Bone - Bone: dyn_bk_strap_02"!(ComponentPose=@n4, bAlphaBoolEnabled:bool=@n3, Translation="0,0,0", Rotation="0,-15,0", Scale="1,1,1") -> (Pose) {
    alphaInputType=Bool
    rotationMode=BMM_Additive
    rotationSpace=BCS_BoneSpace
    translationSpace=BCS_BoneSpace
    alphaBoolBlend {blendOption=Linear}
    boneToModify {boneName=dyn_bk_strap_02}
    rotation {yaw=-15}
   }
   n6 = "Transform (Modify) Bone - Bone: dyn_bk_strap_03"!(ComponentPose=@n5, bAlphaBoolEnabled:bool=@n3, Translation="0,0,0", Rotation="0,-20,0", Scale="1,1,1") -> (Pose) {
    alphaInputType=Bool
    rotationMode=BMM_Additive
    rotationSpace=BCS_BoneSpace
    translationSpace=BCS_BoneSpace
    alphaBoolBlend {blendOption=Linear}
    boneToModify {boneName=dyn_bk_strap_03}
    rotation {yaw=-20}
   }
   n7 = "Transform (Modify) Bone - Bone: root"(ComponentPose=@n6, Translation="0,0.5,0", Rotation="0,0,0") {
    translationMode=BMM_Additive
    translationSpace=BCS_BoneSpace
    alphaBoolBlend {blendOption=Linear}
    boneToModify {boneName=root}
    translation {y=0.5}
   }
   Output Pose(Result=Component To Local(ComponentPose=@n7)) {layerGroup=None}
  ]}
 EventGraph {nodes [
   Event BlueprintUpdateAnimation()
   n1 = Get bIsParachuteOpened()
   > Set Strap_rot-_correction(Strap_rot-_correction:bool=AND Boolean(A=Get bIsSkyDiving(), B=NOT Boolean(A=@n1))) -> (Output_Get:bool)
   TryGetPawnOwner()
  ]}
]
variables [Strap_rot-_correction {type=boolean}]
